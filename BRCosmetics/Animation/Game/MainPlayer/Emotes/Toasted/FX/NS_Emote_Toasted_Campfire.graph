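# Niagara campfire effect for the Toasted emote, layering flame, smoke, ember and light emitters.
# CPU sprite emitters build fire, core glow, rising smoke and curl-noise embers, while one looping particle drives a flickering light and glow sprite.
# GPU SourceParticles inject temperature and velocity into a camera-facing Niagara Fluids 2D gas simulation that renders smoke and fire.
# Both fluid emitters use significance-based instance culling; CoreFire and Embers are distance-culled.
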NiagaraSystem /BRCosmetics/Animation/Game/MainPlayer/Emotes/Toasted/FX/NS_Emote_Toasted_Campfire
dependencies {
 usedDataInterfaces [
  &NiagaraDataInterfaceCurve
  &NiagaraDataInterfaceColorCurve
  &NiagaraDataInterfaceVectorCurve
  &NiagaraDataInterfaceVector2DCurve
  /Script/Niagara.NiagaraDataInterfaceGrid2DCollection
  /Script/Niagara.NiagaraDataInterfaceRenderTarget2D
  &NiagaraDataInterfaceParticleRead
  &NiagaraDataInterfaceCamera
  &NiagaraDataInterfacePlatformSet
  /Script/Niagara.NiagaraDataInterfaceEmitterProperties
 ]
 usedDynamicInputs [&MakeFloatFromVector4]
 usedModules [&SystemState]
}
emitters {
 CoreFire {
  emitterUpdateScript [
   &EmitterState(Enable Distance Culling=true, MaxDistance=2500, Module.bUseMaxDistance=true, Module.bUseMinDistance=false, Scalability Mode=::NewEnumerator1)
   &SpawnRate(SpawnRate=14, Use Spawn Probability=false)
  ]
  particleSpawnScript [
   /Niagara/Modules/Spawn/Initialization/InitializeParticle(
    Lifetime=&Multiply_Float(A=&UniformRangedFloat(Minimum=0.8, Maximum=1.16), B=0.5)
    Module.Write Color=true
    Module.Write Lifetime=true
    Module.Write Mass=true
    Module.Write Material Random=true
    Module.Write Position=false
    Module.Write Scale=false
    Module.Write SpriteRotation=true
    Module.Write SpriteSize=true
    Position=&SimulationPosition()
    Sprite Rotation=&UniformRangedFloat(Minimum=0, Maximum=360)
    Sprite Size=&Vector2DFromFloat(Value=&UniformRangedFloat(Minimum=35, Maximum=50))
   )
   &Color(Color=<rgba 10 0.981 0 0.55>)
   &DynamicMaterialParameters(
    Index 0 Param 2=&RandomRangeFloat(Minimum=0, Maximum=4)
    Index 0 Param 4=2.5
    Module.Param0WriteEnabled=false
    Module.Param1WriteEnabled=true
    Module.Param2WriteEnabled=false
    Module.Param3WriteEnabled=true
   ) {module=DynamicMaterialParameters001}
   &CylinderLocation(
    Cylinder Height=0
    Cylinder Radius=&RandomRangeFloat(Minimum=3, Maximum=5)
    MaskSpawnGroup=false
    Override Seed=false
    UseOwnerScale=true
   )
  ]
  particleUpdateScript [
   /Niagara/Modules/Update/SubUV/SubUVAnimation(Number Of Frames=&FloatFromCurve(FloatCurve=DataInterface() {propertyValues {Curve {keys [{arriveTangent=5.9363, interpMode=RCIM_Cubic, leaveTangent=5.9361, tangentMode=RCTM_User, time=0, value=56}, <at 0.9 60>]}}}))
   &ParticleState()
   &ScaleColor(
    Scale Alpha=&Multiply_Float(
     A=1
     B=&FloatFromCurve(FloatCurve=DataInterface() {propertyValues {Curve {keys [
          {interpMode=RCIM_Cubic, tangentMode=RCTM_User, time=0, value=0}
          {interpMode=RCIM_Cubic, tangentMode=RCTM_User, time=0.35, value=1}
          {arriveTangentWeight=0.3727, interpMode=RCIM_Cubic, tangentMode=RCTM_User, time=1, value=0}
         ]}}})
    )
    Scale RGB=&VectorFromCurve(VectorCurve=DataInterface() {propertyValues {
       XCurve {keys [<at 0.3242 3 RCIM_Cubic>,<at 0.9878 0.9731>]}
       YCurve {keys [<at 0.3242 2.8 RCIM_Cubic>,<at 1 0.9731>]}
       ZCurve {keys [<at 0.3169 2.6605 RCIM_Cubic>,<at 1 0.9731>]}
      }})
    ScaleA=true
    ScaleRGB=true
   ) {module=ScaleColor001}
   &ScaleSpriteSize(Scale Factor=&Vector2DFromCurve(
     Vector2Curve=DataInterface() {propertyValues {XCurve {keys [<at 0 0.5>,<at 1 1>]}, YCurve {keys [<at 0 0.5>,<at 1 1>]}}}
     ScaleCurve=&Vector2DFromFloat(Value=&FN_LOD_Distance_Remap(Start Distance=300, End Distance=1000, Minimum Output=1, Maximum Output=1.1))
    ))
   &DynamicMaterialParameters(
    Index 0 Param 1=&FloatFromCurve(FloatCurve=DataInterface() {propertyValues {Curve {keys [<at 0 1>]}}})
    Index 0 Param 2=0.5
    Index 0 Param 3=&FloatFromCurve(FloatCurve=DataInterface() {propertyValues {Curve {keys [<at 0 0.5 RCIM_Cubic>, {arriveTangent=1.5, interpMode=RCIM_Cubic, leaveTangent=1.5, time=0.5235, value=0.5}, <at 1 2>]}}})
    Module.Param0WriteEnabled=true
    Module.Param1WriteEnabled=false
    Module.Param2WriteEnabled=true
    Module.Param3WriteEnabled=false
   )
   &SpriteRotationRate(Rotation Rate=&Lerp_Float(A=-30, B=30, Alpha=$Particles.Initial.MaterialRandom)) {module=SpriteRotationRate001}
  ]
  rendererClasses [&NiagaraSpriteRendererProperties]
  renderers [{rendererClass=&NiagaraSpriteRendererProperties}]
 }
 Embers {
  emitterUpdateScript [
   &EmitterState(
    Enable Distance Culling=true
    Enable Visibility Culling=true
    Life Cycle Mode=::NewEnumerator1
    Loop Behavior=::NewEnumerator1
    Loop Duration Mode=::NewEnumerator1
    MaxDistance=2000
    Module.bUseMaxDistance=true
    Module.bUseMinDistance=false
    Scalability Mode=::NewEnumerator1
    UseLoopDelay=false
   )
   &SpawnBurst_Instantaneous(Spawn Count=45, Use Spawn Probability=false, UseLoopCountLimit=false)
   &SpawnRate(SpawnRate=25, Use Spawn Probability=false)
  ]
  particleSpawnScript [
   &InitializeParticle(
    Color=<rgba 2 2 2 1>
    Lifetime=&UniformRangedFloat(Minimum=0.3, Maximum=1.5)
    Module.Write Color=true
    Module.Write Lifetime=true
    Module.Write Mass=true
    Module.Write Material Random=true
    Module.Write Position=true
    Module.Write Scale=false
    Module.Write SpriteRotation=false
    Module.Write SpriteSize=true
    Position=&SimulationPosition()
    Sprite Size=&Vector2DFromFloat(Value=&FloatFromCurve(
      FloatCurve=DataInterface() {propertyValues {Curve {keys [<at 0 1.2>,<at 0.1257 0.9847>,<at 0.2996 0.5>,<at 1 0.3>]}}}
      CurveIndex=&UniformRangedFloat(Minimum=0, Maximum=1)
      Scale Curve=2
     ))
   )
   &CylinderLocation(Cylinder Height=5, Cylinder Radius=6, MaskSpawnGroup=false, Override Seed=false, UseOwnerScale=true)
   &AddVelocityFromPoint(Module.Use Falloff Distance=false, Origin Offset=<0 0 -50>, Velocity Strength=50)
   &AddVelocity(
    Cone Angle=250
    Cone Axis=<0 0 1>
    Module.UseConeFalloff=false
    Module.UseVelDistribution=false
    Override Seed=false
    Velocity Mode=::NewEnumerator2
    Velocity Speed=&FloatFromCurve(
     FloatCurve=DataInterface() {propertyValues {Curve {keys [<at 0.1 1 RCIM_Cubic>,<at 0.5 0.1>]}}}
     CurveIndex=$Emitter.Age
     Scale Curve=&RandomRangeFloat(Minimum=120, Maximum=220)
    )
   )
  ]
  particleUpdateScript [
   &ParticleState()
   &AccelerationForce(Acceleration=&Multiply_Vector(
     A=&VectorFromFloat(Value=1)
     B=&UniformRangedVector(Minimum=<0 0 150>, Maximum=<0 0 250>)
    ))
   &Drag(
    Drag=&FloatFromCurve(FloatCurve=DataInterface() {propertyValues {Curve {keys [<at 0 0.5>,<at 0.5 0.5>,<at 1 1>]}}}, Scale Curve=9)
    Use Linear Drag=true
    Use Rotational Drag=false
   )
   /Niagara/Modules/Update/Forces/CurlNoiseForce(
    Noise Frequency=0.04
    Noise Strength=&FloatFromCurve(FloatCurve=DataInterface() {propertyValues {Curve {keys [<at 0 1>,<at 1 0>]}}}, Scale Curve=500)
    Pan Noise=true
    Pan Noise Field=<0 0 1>
   )
   &SolveForcesAndVelocity(Clamp Velocity=false, Limit Acceleration=false)
   &ScaleSpriteSizeBySpeed(
    Max Scale Factor=<2 5>
    Sample Scale Factor By Curve=true
    Scale Factor Curve=DataInterface() {propertyValues {Curve {keys [
        <at -0.0077 0.0765 RCIM_Cubic>
        {arriveTangent=1.2463, interpMode=RCIM_Cubic, leaveTangent=1.2463, tangentMode=RCTM_User, time=0.9985, value=0.9549}
       ]}}}
    Velocity Threshold=500
   )
   &ScaleSpriteSize(Uniform Curve Sprite Scale=DataInterface() {propertyValues {Curve {keys [<at 0.8 1 RCIM_Cubic>, {arriveTangent=-6.6126, interpMode=RCIM_Cubic, leaveTangent=-6.6126, tangentMode=RCTM_User, time=1, value=0}]}}})
   &ScaleColor(
    Scale Alpha=&FloatFromCurve(FloatCurve=DataInterface() {propertyValues {Curve {keys [<at 0 1>,<at 1 0>]}}})
    Scale RGB=&MakeVectorFromLinearColorRGB(Color=&ColorFromCurve(ColorCurve=DataInterface() {propertyValues {
        AlphaCurve {keys [<at 0 1 RCIM_Cubic>]}
        BlueCurve {keys [<at 0 7.8125 RCIM_Cubic>,<at 0.2499 0.7812>,<at 0.7114 0.2083 RCIM_Cubic>]}
        GreenCurve {keys [<at 0 34.1758 RCIM_Cubic>,<at 0.2499 4.8887>,<at 0.7114 0.8934 RCIM_Cubic>]}
        RedCurve {keys [<at 0 50 RCIM_Cubic>,<at 0.2499 10>,<at 0.7114 4 RCIM_Cubic>]}
       }}))
    ScaleA=true
    ScaleRGB=true
   )
  ]
  rendererClasses [&NiagaraSpriteRendererProperties]
  renderers [{rendererClass=&NiagaraSpriteRendererProperties}]
 }
 Fire {
  emitterUpdateScript [
   &EmitterState(Life Cycle Mode=::NewEnumerator1, Loop Behavior=::NewEnumerator1, Loop Duration Mode=::NewEnumerator1, UseLoopDelay=false)
   &SpawnBurst_Instantaneous(Spawn Count=5, Use Spawn Probability=false, UseLoopCountLimit=false)
   &SpawnRate(SpawnRate=16, Use Spawn Probability=false)
  ]
  particleSpawnScript [
   &InitializeParticle(
    Lifetime=&Multiply_Float(A=&UniformRangedFloat(Minimum=0.9, Maximum=1.25), B=0.75)
    Module.Write Color=true
    Module.Write Lifetime=true
    Module.Write Mass=true
    Module.Write Material Random=true
    Module.Write Position=false
    Module.Write Scale=false
    Module.Write SpriteRotation=true
    Module.Write SpriteSize=true
    Position=&SimulationPosition()
    Sprite Rotation=&UniformRangedFloat(Minimum=0, Maximum=360)
    Sprite Size=&Vector2DFromFloat(Value=&FloatFromCurve(
      FloatCurve=DataInterface() {propertyValues {Curve {keys [{arriveTangent=-1.6735, interpMode=RCIM_Cubic, leaveTangent=-1.6735, tangentMode=RCTM_User, time=0, value=1.5}, <at 0.5 1>]}}}
      CurveIndex=$Emitter.Age
      Scale Curve=&UniformRangedFloat(Minimum=35, Maximum=50)
     ))
   )
   &Color(Color=<rgba 10 0.981 0 0.55>)
   &CylinderLocation(
    Cylinder Height=1
    Cylinder Radius=&RandomRangeFloat(Minimum=2, Maximum=3)
    MaskSpawnGroup=false
    Non Uniform Scale:Vector3f=<1 1 1.5>
    Override Seed=false
    UseOwnerScale=true
   )
   &DynamicMaterialParameters(
    Index 0 Param 2=&RandomRangeFloat(Minimum=0, Maximum=4)
    Index 0 Param 4=2.5
    Module.Param0WriteEnabled=false
    Module.Param1WriteEnabled=true
    Module.Param2WriteEnabled=false
    Module.Param3WriteEnabled=true
   ) {module=DynamicMaterialParameters001}
  ]
  particleUpdateScript [
   &SubUVAnimation(Number Of Frames=&FloatFromCurve(FloatCurve=DataInterface() {propertyValues {Curve {keys [{arriveTangent=5.9363, interpMode=RCIM_Cubic, leaveTangent=5.9361, tangentMode=RCTM_User, time=0, value=43}, <at 0.9 62>]}}}))
   &ParticleState()
   &AccelerationForce(Acceleration=&UniformRangedVector(Minimum=<-20 -20 147>, Maximum=<20 20 250>))
   &Drag(Drag=3, Use Linear Drag=true, Use Rotational Drag=false)
   &ScaleColor(
    Scale Alpha=&Multiply_Float(
     A=1
     B=&FloatFromCurve(FloatCurve=DataInterface() {propertyValues {Curve {keys [
          {arriveTangent=2.4553, interpMode=RCIM_Cubic, leaveTangent=2.4553, time=0, value=0}
          {arriveTangent=-0.0116, interpMode=RCIM_Cubic, leaveTangent=-0.0116, tangentMode=RCTM_User, time=0.4073, value=1}
          {arriveTangent=-2.0743, interpMode=RCIM_Cubic, leaveTangent=-2.0743, tangentMode=RCTM_User, time=0.7, value=0.2}
          {arriveTangent=-0.6667, interpMode=RCIM_Cubic, leaveTangent=-0.6667, time=1, value=0}
         ]}}})
    )
    Scale RGB=&VectorFromCurve(VectorCurve=DataInterface() {propertyValues {
       XCurve {keys [<at 0.3242 3 RCIM_Cubic>,<at 0.9878 0.9731>]}
       YCurve {keys [<at 0.3242 2.8 RCIM_Cubic>,<at 1 0.9731>]}
       ZCurve {keys [<at 0.3169 2.6605 RCIM_Cubic>,<at 1 0.9731>]}
      }})
    ScaleA=true
    ScaleRGB=true
   ) {module=ScaleColor001}
   &ScaleSpriteSize(Scale Factor=&Vector2DFromCurve(
     Vector2Curve=DataInterface() {propertyValues {
       XCurve {keys [<at 0 0.3 RCIM_Cubic>, {arriveTangent=0.3, interpMode=RCIM_Cubic, leaveTangent=0.3, time=0.3, value=1}, <at 1 0.6>]}
       YCurve {keys [<at 0 0.3 RCIM_Cubic>, {arriveTangent=0.3, interpMode=RCIM_Cubic, leaveTangent=0.3, time=0.3, value=1}, <at 1 0.6>]}
      }}
     ScaleCurve=<1 1>
    ))
   &DynamicMaterialParameters(
    Index 0 Param 1=&FloatFromCurve(FloatCurve=DataInterface() {propertyValues {Curve {keys [<at 0 1>]}}})
    Index 0 Param 2=0.5
    Index 0 Param 3=&FloatFromCurve(FloatCurve=DataInterface() {propertyValues {Curve {keys [<at 0 0.5 RCIM_Cubic>, {arriveTangent=1.5, interpMode=RCIM_Cubic, leaveTangent=1.5, time=0.5235, value=0.5}, <at 1 2>]}}})
    Module.Param0WriteEnabled=true
    Module.Param1WriteEnabled=false
    Module.Param2WriteEnabled=true
    Module.Param3WriteEnabled=false
   )
   &SpriteRotationRate(Rotation Rate=&Lerp_Float(A=-10, B=10, Alpha=$Particles.Initial.MaterialRandom)) {module=SpriteRotationRate001}
   &SolveForcesAndVelocity(Clamp Velocity=false, Limit Acceleration=false)
  ]
  rendererClasses [&NiagaraSpriteRendererProperties]
  renderers [{rendererClass=&NiagaraSpriteRendererProperties}]
 }
 Grid2D_Gas_SmokeFire {
  simTarget=GPUComputeSim
  emitterSpawnScript [
   /SetVariables(Emitter.ResX=196, Emitter.ResY=196) {isSetParameters=true}
   /SetVariables(Emitter.SimGrid=DataInterface() {propertyValues {
      ClearBeforeNonIterationStage=true
      NumAttributes=0
      NumCellsMaxAxis=3
      NumCellsX=1
      NumCellsY=1
      OverrideBufferFormat=Float
      PreviewAttribute=None
      SetGridFromMaxAxis=false
      bOverrideFormat=false
      bPreviewGrid=false
      RenderTargetUserParameter {parameter {name=None, typeDefHandle {registeredTypeIndex=1}, varData []}}
      WorldBBoxSize=<1200 1200>
     }}) {isSetParameters=true}
   /SetVariables(Emitter.LightingGrid=DataInterface() {propertyValues {
      ClearBeforeNonIterationStage=true
      NumAttributes=3
      NumCellsMaxAxis=3
      NumCellsX=196
      NumCellsY=196
      OverrideBufferFormat=Float
      PreviewAttribute=None
      SetGridFromMaxAxis=false
      bOverrideFormat=false
      bPreviewGrid=false
      RenderTargetUserParameter {parameter {name=None, typeDefHandle {registeredTypeIndex=4}, varData []}}
      WorldBBoxSize=<100 100>
     }}) {isSetParameters=true}
   /SetVariables(Emitter.LightingGridDownsample=DataInterface() {propertyValues {
      ClearBeforeNonIterationStage=true
      NumAttributes=1
      NumCellsMaxAxis=3
      NumCellsX=64
      NumCellsY=64
      OverrideBufferFormat=Float
      PreviewAttribute=None
      SetGridFromMaxAxis=false
      bOverrideFormat=false
      bPreviewGrid=false
      RenderTargetUserParameter {parameter {name=None, typeDefHandle {registeredTypeIndex=4}, varData []}}
      WorldBBoxSize=<100 100>
     }}) {isSetParameters=true}
   /SetVariables() {isSetParameters=true}
   /NiagaraFluids/Modules/Grid2D/Grid2D_SetResolution(Grid:NiagaraDataInterfaceGrid2DCollection=$Emitter.SimGrid, NumCellsX=200, NumCellsY=255, WorldGridExtents:Vector2f=<200 250>)
   &Grid2D_SetResolution(
    Grid:NiagaraDataInterfaceGrid2DCollection=$Emitter.PressureGrid
    OtherGrid:NiagaraDataInterfaceGrid2DCollection=$Emitter.SimGrid
    Resolution Method:ENiagaraGrid2DResolution=::NewEnumerator3
    WorldGridExtents:Vector2f=$Emitter.Grid2D_SetResolution.WorldGridExtents
   ) {module=Grid2D_SetResolution001}
   &Grid2D_SetResolution(
    Grid:NiagaraDataInterfaceGrid2DCollection=$Emitter.LightingGridDownsample
    OtherGrid:NiagaraDataInterfaceGrid2DCollection=$Emitter.SimGrid
    Resolution Method:ENiagaraGrid2DResolution=::NewEnumerator3
    Resolution Mult=0.2
    WorldGridExtents:Vector2f=$Emitter.Grid2D_SetResolution.WorldGridExtents
   ) {module=Grid2D_SetResolution002}
   &Grid2D_SetResolution(
    Grid:NiagaraDataInterfaceGrid2DCollection=$Emitter.LightingGrid
    OtherGrid:NiagaraDataInterfaceGrid2DCollection=$Emitter.SimGrid
    Resolution Method:ENiagaraGrid2DResolution=::NewEnumerator3
    WorldGridExtents:Vector2f=$Emitter.Grid2D_SetResolution.WorldGridExtents
   ) {module=Grid2D_SetResolution003}
   /SetVariables(Emitter.dt=0.0167) {isSetParameters=true}
   /SetVariables(Emitter.g:Vector2f=<0 -9.8>) {isSetParameters=true}
   /SetVariables() {isSetParameters=true}
   /SetVariables(Emitter.density=1) {isSetParameters=true}
   /SetVariables(Emitter.ParticleSource=DataInterface() {propertyValues {EmitterBinding {bindingMode=Other, emitterName=SourceParticles}}}) {isSetParameters=true}
   /SetVariables() {isSetParameters=true}
   &self:SetLightingRT(LightingGrid:NiagaraDataInterfaceGrid2DCollection=$Emitter.LightingGrid, LightingRT:NiagaraDataInterfaceRenderTarget2D=$Emitter.LightingRT) {module=SetLightingRT}
   /NiagaraFluids/Modules/Grid2D/Grid2D_UpdateSimAttributes(Grid:NiagaraDataInterfaceGrid2DCollection=$Emitter.SimGrid, WorldGridSize:Vector2f=$Emitter.Grid2D_SetResolution.WorldGridExtents) {module=UpdateSimAttributes}
   /SetVariables() {isSetParameters=true}
   /SetVariables(Emitter.testCurve=DataInterface() {propertyValues {
      bExposeCurve=true
      AlphaCurve {keys []}
      BlueCurve {keys [<at 0.3032 1>,<at 0.6671 0>]}
      GreenCurve {keys [<at 0.3032 1>,<at 0.6671 0.0785>]}
      RedCurve {keys [<at 0.3032 1>,<at 0.6671 1>]}
     }}) {isSetParameters=true}
   /NiagaraFluids/Functions/General/Fluids_Gas_Manager(
    Density Bouyancy=2
    Density Color Curve=DataInterface() {propertyValues {
      ExposedName=Curve Density
      bExposeCurve=true
      AlphaCurve {keys [<at 0 0 RCIM_Cubic>,<at 1 1>]}
      BlueCurve {keys [<at 0 0.2>,<at 1 0.15>]}
      GreenCurve {keys [<at 0 0.1906>,<at 1 0.143>]}
      RedCurve {keys [<at 0 0.1885>,<at 1 0.1414>]}
     }}
    Density Scale=0.35
    Depth Fade=25
    Dissipation Temperature=0.4
    Dissipation Velocity=0.75
    EnvironmentColor:LinearColor=<rgba 0.1427 0.2031 0.3542 1>
    Pressure Relaxation=1
    Shadowing=2
    SunColor:LinearColor=<rgba 1 0.7033 0.5718 1>
    Temperature Color Curve=DataInterface() {propertyValues {
      ExposedName=Curve Temperature
      bExposeCurve=true
      AlphaCurve {keys [<at 0 0>,<at 0.3256 1>]}
      BlueCurve {keys [<at 0.0696 0>,<at 0.5813 0.0517>,<at 0.9917 0.0783>]}
      GreenCurve {keys [<at 0.0696 0>,<at 0.5813 0.729>,<at 0.9917 0.2349>]}
      RedCurve {keys [<at 0.0696 0>,<at 0.5813 8>,<at 0.9917 0.9401>]}
     }}
    Vorticity Confinement=0.1
   )
   /NiagaraFluids/Modules/Grid2D/Grid2D_InitRT(Grid:NiagaraDataInterfaceGrid2DCollection=$Emitter.SimGrid, SimRT:NiagaraDataInterfaceRenderTarget2D=$Emitter.SimRT) {module=Grid2D_Gas_InitRT}
   &self:Grid2D_Gas_SmokeFire_0.NiagaraScratchPadContainer_2.ScratchModule_0(Grid:NiagaraDataInterfaceGrid2DCollection=$Emitter.SimGrid, PressureGrid:NiagaraDataInterfaceGrid2DCollection=$Emitter.PressureGrid) {module=ScratchModule}
  ]
  emitterUpdateScript [
   /NiagaraFN/Modules/Scalability/SignificantInstancesCull(Cull Response:ENiagaraExecutionStateManagement=::SleepAndLetParticlesFinish, First Frame Only=false)
   &EmitterState(Life Cycle Mode=::NewEnumerator1, Loop Duration=1, Scalability Mode=::NewEnumerator1, UseLoopDelay=false)
   /SetVariables(Emitter.dt:NiagaraFloat=$Engine.DeltaTime) {isSetParameters=true}
   /NiagaraFluids/Modules/Grid2D/Grid2D_CreateUnitToWorldTransform(
    Camera:NiagaraDataInterfaceCamera=$Emitter.CameraQuery
    Camera Facing=true
    Local Pivot:Vector2f=<0.5 0.1>
    Offset:NiagaraPosition=$Engine.Owner.Position
    Rotation:Quat4f=$Engine.Owner.Rotation
    WorldGridExtents:Vector2f=$Emitter.Grid2D_SetResolution.WorldGridExtents
   )
   /SetVariables(Emitter.Position:NiagaraPosition=&ConvertVectorToPosition(
     Input Position=$Emitter.Grid2D_CreateUnitToWorldTransform.LocalToWorldTranslation
     Conversion Operation:ENiagara_LWCConvertVecToPos=::NewEnumerator0
    )) {isSetParameters=true}
   /NiagaraFluids/Modules/Grid2D/Grid2D_SetBounds(
    CameraFacing:NiagaraBool=$Emitter.Grid2D_CreateUnitToWorldTransform.CameraFacing
    EmitterProperties=DataInterface() {propertyValues {EmitterBinding {bindingMode=Self, emitterName=None}}}
    LocalPivot:Vector2f=$Emitter.Grid2D_CreateUnitToWorldTransform.LocalPivot
    Rotate for Camera Facing=true
    WorldGridExtents:Vector2f=$Emitter.Grid2D_SetResolution.WorldGridExtents
   )
  ]
  rendererClasses [&NiagaraSpriteRendererProperties]
  renderers [
   {rendererClass=&NiagaraSpriteRendererProperties}
   {rendererClass=&NiagaraSpriteRendererProperties, rendererIndex=1}
  ]
 }
 Light {
  emitterUpdateScript [
   &EmitterState(
    Enable Distance Culling=true
    Enable Visibility Culling=true
    Inactive Response=::NewEnumerator1
    Life Cycle Mode=::NewEnumerator1
    Loop Behavior=::NewEnumerator1
    Loop Duration Mode=::NewEnumerator1
    Module.bUseMaxDistance=true
    Module.bUseMinDistance=false
    Scalability Mode=::NewEnumerator1
    UseLoopDelay=false
   )
   &SpawnBurst_Instantaneous(Use Spawn Probability=false, UseLoopCountLimit=false)
  ]
  particleSpawnScript [
   /Niagara/Modules/Spawn/Initialization/V2/InitializeParticle(
    Color=<rgba 10 3.2772 0.8854 0.1>
    Lifetime=5
    Position Offset=<0 0 14>
    Sprite Size Mode=::NewEnumerator3
    Uniform Sprite Size=75
    UsePositionOffset=true
    Write Material Random=true
   )
   &DynamicMaterialParameters(
    Index 0 Param 1=5
    Index 0 Param 2=1
    Module.Param0WriteEnabled=true
    Module.Param1WriteEnabled=true
    Module.Param2WriteEnabled=false
    Module.Param3WriteEnabled=false
   )
   &CameraOffset(Camera Offset Amount=20)
  ]
  particleUpdateScript [
   &ParticleState(
    Kill Particles When Lifetime Has Elapsed=false
    Let Infinitely Lived Particles Die When Emitter Deactivates=true
    Loop Particles Lifetime=true
   )
   /SetVariables(Particles.Jitter:NiagaraFloat=&Waveform("Waveform [1]"=::NewEnumerator12, "[1] Frequency"=12, "[1] Amplitude Min/Max"=<0.85 1.1>)) {isSetParameters=true}
   &ScaleSpriteSize(Scale Sprite Size Mode=::NewEnumerator0, Uniform Scale Factor=$Particles.Jitter)
   /SetVariables(Particles.LightRadius=&Multiply_Float(A=$Particles.Jitter, B=150)) {isSetParameters=true}
   &ScaleColor(
    Scale Alpha=&FloatFromCurve(
     FloatCurve=DataInterface() {propertyValues {Curve {keys [
         {arriveTangent=104.8297, interpMode=RCIM_Cubic, leaveTangent=104.8297, tangentMode=RCTM_User, time=0, value=0}
         {arriveTangent=-8.95, interpMode=RCIM_Cubic, leaveTangent=-8.95, tangentMode=RCTM_User, time=0.1, value=3}
         <at 0.5 1>
        ]}}}
     CurveIndex=$System.Age
     Scale Curve=$Particles.Jitter
    )
    ScaleA=true
    ScaleRGB=false
   )
   /SetVariables(Particles.LightColor=&MultiplyLinearColor(Linear Color=$Particles.Color, Scale Factor=&MakeVector4_FromFloat(Float=2))) {isSetParameters=true}
  ]
  rendererClasses [&NiagaraLightRendererProperties, &NiagaraSpriteRendererProperties]
  renderers [
   {rendererClass=&NiagaraLightRendererProperties, rendererIndex=0}
   {rendererClass=&NiagaraSpriteRendererProperties, rendererIndex=1}
  ]
 }
 Smoke {
  emitterUpdateScript [
   &EmitterState(Life Cycle Mode=::NewEnumerator1, Loop Behavior=::NewEnumerator1, Loop Duration Mode=::NewEnumerator1, UseLoopDelay=false)
   &SpawnBurst_Instantaneous(Spawn Count=3, Use Spawn Probability=false, UseLoopCountLimit=false)
   &SpawnRate(SpawnRate=6, Use Spawn Probability=false)
  ]
  particleSpawnScript [
   /SetVariables(Particles.RandomFloat=&UniformRangedFloat(Minimum=0, Maximum=1)) {isSetParameters=true}
   /Niagara/Modules/Spawn/Initialization/InitializeParticle(
    Color=<rgba 0.099 0.099 0.099 1>
    Lifetime=&UniformRangedFloat(Minimum=1.5, Maximum=2)
    Module.Write Color=true
    Module.Write Lifetime=true
    Module.Write Mass=true
    Module.Write Material Random=true
    Module.Write Position=false
    Module.Write Scale=false
    Module.Write SpriteRotation=true
    Module.Write SpriteSize=true
    Position=&SimulationPosition()
    Sprite Rotation=&UniformRangedFloat(Minimum=0, Maximum=360)
    Sprite Size=&Vector2DFromFloat(Value=&FloatFromCurve(
      FloatCurve=DataInterface() {propertyValues {Curve {keys [<at 0 2 RCIM_Cubic>,<at 0.5 1>]}}}
      CurveIndex=$Emitter.Age
      Scale Curve=&UniformRangedFloat(Minimum=16, Maximum=32)
     ))
   )
   &AddVelocity(Velocity=/Niagara/DynamicInputs/Random/RandomVector(
     Vector Scale:NiagaraFloat=25
     Constraint Vector:Vector3f=<0 0 1>
     Randomness Mode:ENiagaraRandomnessMode=::NewEnumerator0
     Random Seed:NiagaraInt32=0
     Fixed Random Seed:NiagaraBool=false
    ))
   &CylinderLocation(Cylinder Height=0, Cylinder Height Midpoint=1, Cylinder Radius=5, MaskSpawnGroup=false, Override Seed=false, UseOwnerScale=true)
   &DynamicMaterialParameters(
    Index 0 Param 4=0.25
    Index 2 Param 2=&RandomRangeFloat(Minimum=0, Maximum=1)
    Module.Param0WriteEnabled=false
    Module.Param0WriteEnabled2=false
    Module.Param1WriteEnabled=false
    Module.Param1WriteEnabled2=true
    Module.Param2WriteEnabled=false
    Module.Param2WriteEnabled2=false
    Module.Param3WriteEnabled=true
    Module.Param3WriteEnabled2=false
    Write Parameter Index 2=true
   )
  ]
  particleUpdateScript [
   &ParticleState()
   &AccelerationForce(Acceleration=&UniformRangedVector(Minimum=<0 0 250>, Maximum=<0 0 350>))
   &Drag(Drag=4, Use Linear Drag=true, Use Rotational Drag=false)
   &DynamicMaterialParameters(
    Index 0 Param 1=&Add_Float(
     A=$Particles.RandomFloat
     B=&FloatFromCurve(FloatCurve=DataInterface() {propertyValues {Curve {keys [
          {arriveTangent=2.4104, time=0, value=1}
          {arriveTangent=0.2537, interpMode=RCIM_Cubic, leaveTangent=0.2537, tangentMode=RCTM_User, time=1, value=2}
         ]}}})
    )
    Index 0 Param 4=1
    Index 2 Param 1=&FloatFromCurve(FloatCurve=DataInterface() {propertyValues {Curve {keys [{interpMode=RCIM_Cubic, leaveTangent=0.9717, tangentMode=RCTM_Break, time=0.25, value=0.8}, <at 1 1>]}}})
    Index 2 Param 2=&FloatFromCurve(
     FloatCurve=DataInterface() {propertyValues {Curve {keys [<at 0 0.5 RCIM_Cubic>,<at 1 0 RCIM_Cubic>]}}}
     Scale Curve=&RandomRangeFloat(Minimum=0.2, Maximum=0.5)
    )
    Index 2 Param 3=&FloatFromCurve(FloatCurve=DataInterface() {propertyValues {Curve {keys [
         {interpMode=RCIM_Cubic, tangentMode=RCTM_Break, time=0, value=1}
         {interpMode=RCIM_Cubic, leaveTangent=11.1233, tangentMode=RCTM_Break, time=1, value=3}
        ]}}})
    Index 2 Param 4=&FloatFromCurve(
     FloatCurve=DataInterface() {propertyValues {Curve {keys [
         {arriveTangent=-0.4552, interpMode=RCIM_Cubic, leaveTangent=-0.4552, tangentMode=RCTM_User, time=0, value=1}
         {interpMode=RCIM_Cubic, leaveTangent=-1.2956, tangentMode=RCTM_Break, time=1, value=0.75}
        ]}}}
     Scale Curve=1.4
    )
    Module.Param0WriteEnabled=true
    Module.Param0WriteEnabled2=true
    Module.Param1WriteEnabled=false
    Module.Param1WriteEnabled2=true
    Module.Param2WriteEnabled=false
    Module.Param2WriteEnabled2=true
    Module.Param3WriteEnabled=false
    Module.Param3WriteEnabled2=true
    Write Parameter Index 2=true
   ) {module=DynamicMaterialParameters001}
   &ScaleSpriteSize(Scale Factor=&Vector2DFromFloat(Value=&FloatFromCurve(
      FloatCurve=DataInterface() {propertyValues {Curve {keys [
          {arriveTangent=3.358, interpMode=RCIM_Cubic, leaveTangent=3.358, tangentMode=RCTM_User, time=0, value=0.1773}
          {arriveTangent=0.8227, interpMode=RCIM_Cubic, leaveTangent=0.8227, time=0.1179, value=0.489}
          {interpMode=RCIM_Cubic, leaveTangent=0.0321, time=1, value=1}
         ]}}}
      Scale Curve=2
     )))
   &ScaleColor(
    Scale Alpha=&FloatFromCurve(FloatCurve=DataInterface() {propertyValues {Curve {keys [
         {arriveTangent=9191.0225, interpMode=RCIM_Cubic, leaveTangent=4.1064, tangentMode=RCTM_User, time=0.0001, value=0.0123}
         {arriveTangentWeight=0.0484, interpMode=RCIM_Cubic, leaveTangentWeight=0.2906, tangentMode=RCTM_User, time=0.1, value=1}
         {arriveTangent=-1.1111, interpMode=RCIM_Cubic, leaveTangent=-1.1111, time=0.8, value=0.75}
         <at 1 0 RCIM_Cubic>
        ]}}})
    Scale RGB=&Lerp_Vector(
     A=<1 1 1>
     B=<5 3 1>
     Alpha=&FloatFromCurve(FloatCurve=DataInterface() {propertyValues {Curve {keys [{arriveTangent=-2.6611, interpMode=RCIM_Cubic, leaveTangent=-2.6611, tangentMode=RCTM_User, time=0, value=1}, <at 0.5 0>]}}})
    )
    ScaleA=true
    ScaleRGB=true
   )
   &SpriteRotationRate(Rotation Rate=&Multiply_Float(
     A=&Lerp_Float(A=-150, B=150, Alpha=$Particles.MaterialRandom)
     B=&Lerp_Float(A=1, B=0, Alpha=&Power(a=$Particles.NormalizedAge, b=0.7))
    ))
   &CameraOffset(Camera Offset Amount=21)
   &SolveForcesAndVelocity(Clamp Velocity=false, Limit Acceleration=false)
  ]
  rendererClasses [&NiagaraSpriteRendererProperties]
  renderers [{rendererClass=&NiagaraSpriteRendererProperties}]
 }
 SourceParticles {
  simTarget=GPUComputeSim
  emitterSpawnScript [
   /SetVariables(Emitter.RasterizationGrid=DataInterface() {propertyValues {
      ClearBeforeNonIterationStage=true
      NumAttributes=4
      NumCellsMaxAxis=3
      NumCellsX=3
      NumCellsY=3
      OverrideBufferFormat=Float
      PreviewAttribute=None
      SetGridFromMaxAxis=false
      bOverrideFormat=false
      bPreviewGrid=false
      RenderTargetUserParameter {parameter {name=None, typeDefHandle {registeredTypeIndex=4}, varData []}}
      WorldBBoxSize=<100 100>
     }}) {isSetParameters=true}
   &Grid2D_SetResolution(Grid:NiagaraDataInterfaceGrid2DCollection=$Emitter.RasterizationGrid, NumCellsX=256, NumCellsY=256, WorldGridExtents:Vector2f=<1200 1200>)
   &Grid2D_UpdateSimAttributes(Grid:NiagaraDataInterfaceGrid2DCollection=$Emitter.RasterizationGrid, WorldGridSize:Vector2f=$Emitter.Grid2D_SetResolution.WorldGridExtents) {module=UpdateSimAttributes}
  ]
  emitterUpdateScript [
   &SignificantInstancesCull(Cull Response:ENiagaraExecutionStateManagement=::SleepAndLetParticlesFinish, First Frame Only=false, Num Instances=2)
   &EmitterState(Life Cycle Mode=::NewEnumerator1, Loop Duration=0.8, Scalability Mode=::NewEnumerator1, UseLoopDelay=false)
   &SpawnBurst_Instantaneous(
    Spawn Count=&RandomRangeInt(Minimum=100, Maximum=100, Evaluation Type=::NewEnumerator1)
    Use Spawn Probability=false
   ) {enabled=false}
   &SpawnRate(SpawnRate=200, Use Spawn Probability=false)
   &Grid2D_CreateUnitToWorldTransform(
    Camera Facing=true
    Local Pivot:Vector2f=<0.5 0.1>
    Offset:NiagaraPosition=$Engine.Owner.Position
    Rotation:Quat4f=$Engine.Owner.Rotation
    WorldGridExtents:Vector2f=$Emitter.Grid2D_SetResolution.WorldGridExtents
   )
  ]
  particleSpawnScript [
   &InitializeParticle(
    Color=<rgba 1111 13.6009 0 1>
    Lifetime=&RandomRangeFloat(Minimum=0.09, Maximum=0.125)
    Mass=&RandomRangeFloat(Minimum=0.75, Maximum=1)
    Module.Write Color=true
    Module.Write Lifetime=true
    Module.Write Mass=true
    Module.Write Material Random=true
    Module.Write Position=false
    Module.Write Scale=false
    Module.Write SpriteRotation=false
    Module.Write SpriteSize=true
    Sprite Size=&Vector2DFromFloat(Value=&RandomRangeFloat(Minimum=12, Maximum=250))
   )
   &AddVelocityInCone(
    Cone Angle=&Cosine(Normalized Angle=$Emitter.Age, Period=0.4, Scale=0, Bias=88)
    Cone Axis=<0 0 1>
    Cone Axis Coordinate Space=::World
    Module.Use Velocity Falloff On Cone Axis=true
    Velocity Falloff Away From Cone Axis=0.333
    Velocity Strength=&RandomRangeFloat(Minimum=12, Maximum=350)
   ) {enabled=false}
   &SphereLocation(
    MaskSpawnGroup=false
    Non Uniform Scale=<1 1 0.4>
    Offset=&RandomRangeVector(Minimum=<0 0 -5>, Maximum=<1 1 15>)
    Override Seed=false
    Sphere Radius=&Lerp_Float(A=2, B=5, Alpha=&Sine(Normalized Angle=$Emitter.Age, Period=4, Scale=0.5, Bias=1))
    Surface Only=false
    UseNonUniformScale=true
    UseOffset=true
    UseSphereOrientation=false
   )
   &CurlNoiseForce(Noise Frequency=0.04, Noise Strength=40, Pan Noise=true, Pan Noise Field=<0.3 0.5 0.3>)
   &ApplyInitialForces()
  ]
  particleUpdateScript [
   &ParticleState()
   &CurlNoiseForce(Noise Frequency=0.03, Noise Strength=100, Pan Noise=true, Pan Noise Field=<0.1 0.5 0.2>) {enabled=false, module=CurlNoiseForce001}
   &GravityForce(Gravity=<0 0 220>)
   &Drag(Use Linear Drag=true, Use Rotational Drag=false)
   &ScaleColor(
    Scale Alpha=&FloatFromCurve(FloatCurve=DataInterface() {propertyValues {Curve {keys [<at -0.0046 1.0113 RCIM_Cubic>,<at 1.0011 0.002 RCIM_Cubic>]}}})
    ScaleA=true
    ScaleRGB=true
   )
   &Collision(
    Age Colliding Particles=false
    Enable Rest State=false
    Friction=0
    GPU Collision Type=::NewEnumerator2
    Module.EnableMaxCollisionCount=false
    Module.Randomize Collision Normal Vector=false
    Restitution=0.7
   ) {enabled=false, module=Collision001}
   &SolveForcesAndVelocity(Clamp Velocity=false, Limit Acceleration=false)
   /NiagaraFluids/Modules/Particles/Fluids_Gas_Source(
    Falloff=0.1
    NoiseIntensity=5.25
    NoiseScale=12
    Radius:NiagaraFloat=&FloatFromCurve(
     FloatCurve=DataInterface() {propertyValues {Curve {keys [<at 0 0>,<at 0.089 0.6165>,<at 0.1628 0.4902>,<at 0.2318 0.984>,<at 0.317 0.2106>,<at 0.3821 0.6886>,<at 1 0>]}}}
     Scale Curve=&RandomRangeFloat(Minimum=2, Maximum=4)
    )
    Temperature:NiagaraFloat=&FloatFromCurve(
     FloatCurve=DataInterface() {propertyValues {Curve {keys [
         {interpMode=RCIM_Cubic, leaveTangent=1.8647, tangentMode=RCTM_Break, time=0, value=0.5}
         {interpMode=RCIM_Cubic, tangentMode=RCTM_Break, time=0.228, value=0.772}
         <at 1 0>
        ]}}}
     Scale Curve=&RandomRangeFloat(Minimum=200, Maximum=700)
    )
    VelocityScale:NiagaraFloat=&RandomRangeFloat(Minimum=1, Maximum=2.5)
   )
   /SetVariables(Particles.SpriteSize=&Vector2DFromFloat(Value=&Multiply_Float(A=$Particles.fluids_source_radius, B=2))) {isSetParameters=true}
  ]
  rendererClasses [&NiagaraSpriteRendererProperties]
  renderers [{rendererClass=&NiagaraSpriteRendererProperties}]
 }
}
summary {
 systemName=NS_Emote_Toasted_Campfire
 emitters [
  {emitterName=Smoke, rendererClasses [&NiagaraSpriteRendererProperties]}
  {emitterName=Embers, rendererClasses [&NiagaraSpriteRendererProperties]}
  {emitterName=Fire, rendererClasses [&NiagaraSpriteRendererProperties]}
  {emitterName=CoreFire, rendererClasses [&NiagaraSpriteRendererProperties]}
  {emitterName=Light, rendererClasses [&NiagaraLightRendererProperties, &NiagaraSpriteRendererProperties]}
  {emitterName=Grid2D_Gas_SmokeFire, simTarget=GPUComputeSim, rendererClasses [&NiagaraSpriteRendererProperties]}
  {emitterName=SourceParticles, simTarget=GPUComputeSim, rendererClasses [&NiagaraSpriteRendererProperties]}
 ]
 userVariables []
}
